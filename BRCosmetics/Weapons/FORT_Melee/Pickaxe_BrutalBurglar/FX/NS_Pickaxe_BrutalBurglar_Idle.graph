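# Idle Niagara effect for the Brutal Burglar pickaxe that shows sparse, warm-tinted sprite sparkles on the weapon.
# A single Sparkles emitter spawns at a low, probability-gated rate, using a ten-way particle partition to index a hand-placed array of 40 local offsets.
# Each camera-offset sprite spins while curves grow, shrink and fade it over its life.
# A time-of-day scalar scales RGB, dimming the sparkles in evening and at night.
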NiagaraSystem /BRCosmetics/Weapons/FORT_Melee/Pickaxe_BrutalBurglar/FX/NS_Pickaxe_BrutalBurglar_Idle
dependencies {
 usedDataInterfaces [&NiagaraDataInterfaceArrayFloat3, &NiagaraDataInterfaceCurve]
 usedDynamicInputs []
 usedModules [&SystemState]
}
emitters {Sparkles {
  emitterUpdateScript [
   &EmitterState()
   &SpawnRate(Spawn Probability=0.25, SpawnRate=7, Use Spawn Probability=true)
  ]
  particleSpawnScript [
   &PartitionParticles(Number Of Partitions=10)
   &InitializeParticle(
    Color=<rgba 1 0.6026 0.2024 1>
    Sprite Size Mode=::NewEnumerator1
    Uniform Sprite Size Max=20
    Uniform Sprite Size Min=15
    UsePositionOffset=false
    Write Material Random=true
   )
   &SystemLocation(
    MaskSpawnGroup=false
    Offset=&SelectVectorFromArray(
     Direct Array Index=$Particles.PartitionParticles.Partition
     Vector Selection Array=DataInterface() {propertyValues {FloatData [
        <-0.1511 5.3239 39.9528>
        <0.4829 -0.5035 28.6536>
        <0.8245 3.1917 43.356>
        <0.2081 -1.4435 54.9687>
        <-0.1316 10.2868 33.1988>
        <0.8369 1.0808 23.1696>
        <-0.4372 -2.6058 8.7994>
        <0.9309 0.0629 11.9196>
        <-0.6387 3.8853 36.2382>
        <0.6249 3.8428 38.2699>
        <-0.6178 0.3921 20.6545>
        <0.6515 2.6474 30.5554>
        <0.6234 3.5681 40.792>
        <-0.2926 5.6907 30.4199>
        <0.856 0.7324 16.4124>
        <0.5691 -2.0272 35.7189>
        <0.785 1.1876 19.2615>
        <0.799 6.9869 35.9312>
        <-0.1074 -3.0754 7.0186>
        <0.7601 -1.2407 9.5602>
        <0.3177 1.8798 21.6565>
        <-0.817 3.5499 44.2656>
        <0.8457 6.1333 32.2043>
        <0.9803 3.1154 32.659>
        <0.4387 3.6569 27.3175>
        <-0.3002 1.9301 49.4859>
        <0.7659 7.664 33.047>
        <-1.107 4.7839 34.5243>
        <-0.5055 0.162 31.0399>
        <-0.1793 -3.835 56.7222>
        <0.5844 2.0762 46.1265>
        <0 9.2858 34.0885>
        <-0.8369 1.9585 24.9487>
        <-0.6551 5.7354 38.3001>
        <0.3748 4.633 42.1638>
        <-0.3811 -2.2327 5.6375>
        <-0.8373 0.5915 18.4254>
        <0.6831 1.2589 34.4353>
        <-0.8126 -1.3742 10.9504>
        <-0.7914 -0.2341 35.4955>
       ]}}
    )
   )
   &CameraOffset(Camera Offset Amount=2)
  ]
  particleUpdateScript [
   &ParticleState()
   &SpriteRotationRate(Rotation Rate=75)
   &ScaleSpriteSize(Uniform Curve Sprite Scale=DataInterface() {propertyValues {Curve {keys [
        {arriveTangent=4.7337, interpMode=RCIM_Cubic, leaveTangent=4.7337, tangentMode=RCTM_User, time=0, value=0}
        {interpMode=RCIM_Cubic, tangentMode=RCTM_User, time=0.5, value=1}
        {tangentMode=RCTM_User, time=1, value=0}
       ]}}})
   &ScaleColor(
    Scale Alpha=&FloatFromCurve(
     FloatCurve=DataInterface() {propertyValues {Curve {keys [
         {arriveTangent=4.5234, interpMode=RCIM_Cubic, leaveTangent=4.5234, tangentMode=RCTM_User, time=0, value=0}
         {interpMode=RCIM_Cubic, tangentMode=RCTM_User, time=0.5, value=1}
         {tangentMode=RCTM_User, time=1, value=0}
        ]}}}
     Scale Curve=0.25
    )
    Scale RGB=&VectorFromFloat(Value=&FNTimeOfDayScalar(Night Scale=0.5, Morning Scale=1, Evening Scale=0.8))
    ScaleA=true
    ScaleRGB=true
   )
  ]
  rendererClasses [&NiagaraSpriteRendererProperties]
  renderers [{rendererClass=&NiagaraSpriteRendererProperties}]
 }}
summary {
 systemName=NS_Pickaxe_BrutalBurglar_Idle
 emitters [{emitterName=Sparkles, rendererClasses [&NiagaraSpriteRendererProperties]}]
 userVariables []
}
